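# Skeletal-mesh body material for the Star Walker player character, giving it an animated emissive galaxy look.
# Camera-vector triplanar nebula, warped by a glass-crack texture, combines with Fresnel rims, crack glow, a pre-skinned radial gradient and panner-twinkled stars.
# World position offset pushes camera-facing vertices toward the camera, and MF_TechArtOverrides wraps the result.
# A height-blended head mask using scrolling streaks exists, but its output is unconnected.
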Material /Game/Characters/Player/Male/Medium/Bodies/M_MED_Star_Walker/Materials/M_StarWalker_body
graph {
 properties {
  bUseMaterialAttributes=true
  bUsedWithMeshDeformer=true
  bUsedWithMorphTargets=true
  bUsedWithSkeletalMesh=true
  pixelDepthOffsetMode=PDOM_Legacy
  shadingModel=MSM_FromMaterialExpression
  usedShadingModels=Default Lit
 }
 nodes [
  n1 = Fresnel(ExponentIn=ScalarParameter() {defaultValue=0.8362, parameterName=FresnelPower}) {baseReflectFraction=0}
  n2 = Saturate(None=@n1)
  n3 = VectorParameter() {parameterName=FrenselColor, defaultValue=<rgba 0.9434 0.6823 4 1>}
  n4 = ScalarParameter() {defaultValue=1.0039, parameterName=Distort}
  n5 = Time()
  n6 = VectorParameter() {parameterName=Panner SpeedG, defaultValue=<rgba 0.0772 0.0772 0.0772 1>}
  n7 = ScalarParameter() {defaultValue=2, group=Galaxy, parameterName=GalaxyTiling_NearG}
  n8 = ScalarParameter() {defaultValue=4, group=Galaxy, parameterName=GalaxyTiling_FarG}
  n9 = MaterialFunctionCall(Distance=ScalarParameter() {defaultValue=2000, parameterName=Near Camera Tiling DistanceG}) {materialFunction=&Near_Camera_Fade}
  n10 = ComponentMask(None=MaterialFunctionCall!(
    CameraVector
    Rotation Speed=Multiply(A=@n5, B=@n6.RGB)
    Texture=TextureObject() {texture=/Game/Effects/Fort_Effects/Textures/Tile/T_Glass_Cracks_M}
    Tiling=LinearInterpolate(A=@n7, B=@n8, Alpha=@n9.Result)
    Offset=ScalarParameter() {defaultValue=5, group=Galaxy, parameterName=WarpIntensityG}
    AxisFadeContrast=ScalarParameter() {defaultValue=3, parameterName=AxisFadeContrastG}
   ) {materialFunction=/Game/Weapons/WeaponSkins/MaterialFunction/TriplanarCameraVector_Rotated}.Result) {r=true}
  n11 = Multiply(A=Constant3Vector() {constant=<rgba 1 1 1 0>}, B=@n10)
  n12 = ScalarParameter() {defaultValue=1000, parameterName=Near Camera Tiling Distance}
  n13 = MaterialFunctionCall(
   CameraVector=Add(A=Multiply(A=@n4, B=@n11), B=CameraVectorWS())
   Rotation Speed=Multiply(A=Time(), B=VectorParameter() {parameterName=Panner Speed, defaultValue=<rgba 0.125 0.125 -0.0625 1>}.RGB)
   Texture=TextureObject() {texture=/Game/Characters/Player/Male/Medium/Bodies/M_Med_Soldier_04/Skins/Galaxy/Textures/T_FN_Nebula}
   Tiling=LinearInterpolate(
    A=ScalarParameter() {defaultValue=2, group=Galaxy, parameterName=GalaxyTiling_Near}
    B=ScalarParameter() {defaultValue=1, group=Galaxy, parameterName=GalaxyTiling_Far}
    Alpha=MaterialFunctionCall(Distance=@n12) {materialFunction=&Near_Camera_Fade}.Result
   )
   Offset=ScalarParameter() {defaultValue=0.012, group=Galaxy, parameterName=WarpIntensity}
   AxisFadeContrast=ScalarParameter() {defaultValue=6, parameterName=AxisFadeContrast}
  ) {materialFunction=&TriplanarCameraVector_Rotated}
  n14 = Add(
   A=Multiply(A=Multiply(A=@n2, B=@n3.RGB), B=Desaturation(None=@n13.Result))
   B=Multiply(A=Power(Base=@n1) {constExponent=3}, B=VectorParameter() {parameterName=Edge, defaultValue=<rgba 0.1591 0.3847 0.65 1>}.RGB)
  )
  n15 = Multiply(
   A=VectorParameter() {parameterName=ColorG2, defaultValue=<rgba 1.9832 4.9086 8 1>}.RGB
   B=Power(Base=@n10, Exp=ScalarParameter() {defaultValue=2.8616, parameterName=FallofGlass})
  )
  n16 = ScalarParameter() {defaultValue=0.5, parameterName=Distort2}
  n17 = ComponentMask(None=PreSkinnedPosition()) {b=true, r=true}
  n18 = ComponentMask(None=VectorParameter() {parameterName=Center, defaultValue=<rgba 0 0 126.2117 1>}.RGB) {b=true, r=true}
  n19 = VertexInterpolator(VS=Divide(A=Distance(A=@n17, B=@n18), B=ScalarParameter() {defaultValue=80, parameterName=XDistance}))
  n20 = OneMinus(None=Add(A=Multiply(A=@n16, B=@n10), B=ComponentMask(None=@n19.PS) {r=true}))
  n21 = ScalarParameter() {defaultValue=8, parameterName=FresnelPower3}
  n22 = Multiply(A=Saturate(None=Power(Base=@n20, Exp=@n21)), B=VectorParameter() {parameterName=ColorG1d, defaultValue=<rgba 0.1491 0.4393 1 1>}.RGB)
  n23 = LinearInterpolate(
   A=VectorParameter() {parameterName=ColorG1, defaultValue=<rgba 0.0122 0.5449 1 1>}.RGB
   B=VectorParameter() {parameterName=ColorG1c, defaultValue=<rgba 0.3335 0.1875 2 1>}.RGB
   Alpha=Fresnel(ExponentIn=ScalarParameter() {defaultValue=0.2314, parameterName=FresnelPower2}) {baseReflectFraction=0}
  )
  n24 = ComponentMask(None=@n13.Result) {r=true}
  n25 = VectorParameter() {parameterName=ColorG1b, defaultValue=<rgba 0.0816 0.2477 0.3 1>}
  n26 = ComponentMask(None=@n13.Result) {b=true}
  n27 = Multiply(A=LinearInterpolate(A=Multiply(A=@n23, B=@n24), B=Multiply(A=@n25.RGB, B=@n26)), B=ScalarParameter() {defaultValue=1})
  n28 = ScalarParameter() {defaultValue=0.0762, parameterName=BlendOriginal}
  n29 = Time()
  n30 = ScalarParameter() {defaultValue=0.0156, parameterName=Time Scale}
  n31 = TextureObject() {samplerType=SAMPLERTYPE_Masks, texture=/Game/Characters/Player/Male/Medium/Bodies/M_Med_Soldier_04/Skins/Galaxy/Textures/Celestial_T_Stars}
  n32 = ComponentMask(None=MaterialFunctionCall(
    Rotation Speed=Multiply(A=@n29, B=@n30)
    Texture=@n31
    Tiling=ScalarParameter() {defaultValue=4, group=Galaxy, parameterName=MediumStarTiling}
    Offset=ScalarParameter() {defaultValue=0.012, group=Galaxy, parameterName=MediumStar_WarpIntensity}
    AxisFadeContrast=ScalarParameter() {defaultValue=6, parameterName=AxisFadeContrast}
   ) {materialFunction=&TriplanarCameraVector_Rotated}.Result) {b=true}
  n33 = Multiply(A=@n32, B=@n32)
  n34 = TextureSample(UVs=Panner(Coordinate=TextureCoordinate() {uTiling=2, vTiling=2}) {speedX=0.137, speedY=-0.2}) {
   samplerSource=SSM_Wrap_WorldGroupSettings
   samplerType=SAMPLERTYPE_Masks
   texture=&Celestial_T_Stars
  }
  n35 = Time()
  n36 = ScalarParameter() {defaultValue=0.0156, parameterName=Time Scale}
  n37 = TextureObject() {samplerType=SAMPLERTYPE_Masks, texture=&Celestial_T_Stars}
  n38 = ComponentMask(None=MaterialFunctionCall(
    Rotation Speed=Multiply(A=@n35, B=@n36)
    Texture=@n37
    Tiling=ScalarParameter() {defaultValue=2, group=Galaxy, parameterName=LargeStarTiling}
    Offset=ScalarParameter() {defaultValue=0.012, group=Galaxy, parameterName=LargeStar_WarpIntensity}
    AxisFadeContrast=ScalarParameter() {defaultValue=6, parameterName=AxisFadeContrast}
   ) {materialFunction=&TriplanarCameraVector_Rotated}.Result) {r=true}
  n39 = Multiply(A=@n38, B=@n38)
  n40 = TextureSample(UVs=Panner(Coordinate=TextureCoordinate()) {speedX=0.137, speedY=-0.2}) {
   samplerSource=SSM_Wrap_WorldGroupSettings
   samplerType=SAMPLERTYPE_Masks
   texture=&Celestial_T_Stars
  }
  n41 = Add(
   A=Multiply(A=VectorParameter() {group=Galaxy, parameterName=Star Brightness_1, defaultValue=<rgba 10 10 10 1>}.RGB, B=Multiply(A=@n33, B=@n34.G))
   B=Multiply(A=VectorParameter() {group=Galaxy, parameterName=Star Brightness_2, defaultValue=<rgba 10 10 10 20>}.RGB, B=Multiply(A=@n39, B=@n40.G))
  )
  n42 = Add(
   A=Add(A=Add(A=@n15, B=@n22), B=LinearInterpolate(A=@n27, B=@n13.Result, Alpha=@n28))
   B=Multiply(A=VectorParameter() {parameterName=RiftShardColor, defaultValue=<rgba 0.9279 7.6847 10 0>}.RGB, B=Multiply(A=@n13.Result, B=@n41))
  )
  n43 = Constant()
  n44 = Normalize(VectorInput=Subtract(A=CameraPositionWS(), B=WorldPosition().XYZ))
  n45 = Multiply(A=@n44, B=ScalarParameter() {defaultValue=5, parameterName=PushtoCamera})
  n46 = VertexNormalWS()
  n47 = Saturate(None=Power(Base=DotProduct(A=@n46, B=@n44), Exp=ScalarParameter() {defaultValue=0.7, parameterName=WPOFrenellPower}))
  MaterialFunctionCall(In=SetMaterialAttributes!(MaterialAttributes, Emissive Color=Add(A=@n14, B=@n42), Opacity Mask, World Position Offset=LinearInterpolate(A=@n43, B=@n45, Alpha=@n47)) {attributeSetTypes [#1, #2, #3]}) {materialFunction=&MF_TechArtOverrides}
  n48 = PreSkinnedPosition()
  n49 = ComponentMask(None=@n48) {b=true}
  n50 = ScalarParameter() {defaultValue=139.478, parameterName=HeadBlend}
  n51 = Saturate(None=Divide(A=Subtract(A=@n49, B=@n50), B=ScalarParameter() {defaultValue=32.2506, parameterName=HeadBlendDistance}))
  n52 = TextureObject() {texture=/Game/Effects/Fort_Effects/Textures/Energy/T_OutpostShockwaveStreaks_02}
  n53 = ScalarParameter() {defaultValue=10.5326, parameterName=SizeOpacityMask}
  n54 = Time()
  n55 = ScalarParameter() {defaultValue=-6.8878, parameterName=Speed}
  n56 = VertexInterpolator(VS=Add(A=@n48, B=Multiply(A=@n54, B=@n55)))
  MaterialFunctionCall(
   Transition Phase=VertexInterpolator(VS=OneMinus(None=@n51)).PS
   Height Texture=OneMinus(None=MaterialFunctionCall(TextureObject=@n52, TextureSize=@n53, WorldPosition=@n56.PS) {materialFunction=&WorldAlignedTexture}.XY Texture)
  ) {materialFunction=&HeightLerp}
 ]
}
