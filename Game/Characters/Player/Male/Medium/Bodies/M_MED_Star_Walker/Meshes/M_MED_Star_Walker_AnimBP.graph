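# Anim Blueprint for the Star Walker male body, animating its long coat and collar atop the pose copied from the attached parent mesh.
# With rigid bodies enabled, simulations switch between normal, force-driven skydiving and DBNO-carried setups.
# Otherwise, weighted bone constraints pose the coat from spine, pelvis and legs, with knee-bend-driven back corrections.
# The event graph computes alphas and forces, resetting dynamics around skydiving, parachuting and DBNO recovery.
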AnimBlueprint /Game/Characters/Player/Male/Medium/Bodies/M_MED_Star_Walker/Meshes/M_MED_Star_Walker_AnimBP
target_skeleton=&/M_MED_Star_Walker_Skeleton
graphs [
 AnimGraph {nodes [
   n1 = "Constraint - jacket_frt_03_l (PPPPP)"!(
    ComponentPose=Local To Component(LocalPose=Use cached pose 'AnimCache'())
    Alpha=1
    ConstraintWeights_0=0.1633
    ConstraintWeights_1:float=0.3139
    ConstraintWeights_2:float=0.1437
    ConstraintWeights_3:float=0.3058
    ConstraintWeights_4:float=0.0732
   ) -> (Pose) {
    alphaBoolBlend {blendOption=Linear}
    boneToModify {boneName=jacket_frt_03_l}
    constraintSetup [
     {targetBone {boneName=spine_01}}
     {targetBone {boneName=spine_02}}
     {targetBone {boneName=spine_03}}
     {targetBone {boneName=spine_04}}
     {targetBone {boneName=pelvis}}
    ]
    constraintWeights [1,1,1,1,1]
   }
   n2 = "Constraint - jacket_frt_04_l (PP)"!(ComponentPose=@n1, Alpha=1, ConstraintWeights_0=0.7617, ConstraintWeights_1:float=0.2383) -> (Pose) {
    alphaBoolBlend {blendOption=Linear}
    boneToModify {boneName=jacket_frt_04_l}
    constraintSetup [{targetBone {boneName=pelvis}}, {targetBone {boneName=thigh_l}}]
    constraintWeights [1,1]
   }
   n3 = "Constraint - jacket_frt_05_l (PP)"!(ComponentPose=@n2, Alpha=1, ConstraintWeights_0=0.1011, ConstraintWeights_1:float=0.8989) -> (Pose) {
    alphaBoolBlend {blendOption=Linear}
    boneToModify {boneName=jacket_frt_05_l}
    constraintSetup [{targetBone {boneName=pelvis}}, {targetBone {boneName=thigh_l}}]
    constraintWeights [1,1]
   }
   n4 = "Constraint - jacket_frt_06_l (PP)"!(ComponentPose=@n3, Alpha=1, ConstraintWeights_0=0.0113, ConstraintWeights_1:float=0.9887) -> (Pose) {
    alphaBoolBlend {blendOption=Linear}
    boneToModify {boneName=jacket_frt_06_l}
    constraintSetup [{targetBone {boneName=pelvis}}, {targetBone {boneName=thigh_l}}]
    constraintWeights [1,1]
   }
   n5 = "Constraint - jacket_frt_07_l (PP)"!(ComponentPose=@n4, Alpha=1, ConstraintWeights_0=0.0191, ConstraintWeights_1:float=0.9809) -> (Pose) {
    alphaBoolBlend {blendOption=Linear}
    boneToModify {boneName=jacket_frt_07_l}
    constraintSetup [{targetBone {boneName=calf_l}}, {targetBone {boneName=thigh_l}}]
    constraintWeights [1,1]
   }
   n6 = "Constraint - jacket_frt_08_l (PP)"!(ComponentPose=@n5, Alpha=1, ConstraintWeights_0=0.963, ConstraintWeights_1:float=0.037) -> (Pose) {
    alphaBoolBlend {blendOption=Linear}
    boneToModify {boneName=jacket_frt_08_l}
    constraintSetup [{targetBone {boneName=calf_l}}, {targetBone {boneName=thigh_l}}]
    constraintWeights [1,1]
   }
   n7 = "Constraint - jacket_frt_09_l (P)"!(ComponentPose=@n6, Alpha=1, ConstraintWeights_0=1, ConstraintWeights_1:float=0) -> (Pose) {
    alphaBoolBlend {blendOption=Linear}
    boneToModify {boneName=jacket_frt_09_l}
    constraintSetup [{targetBone {boneName=calf_l}}]
    constraintWeights [1,0]
   }
   n8 = "Constraint - jacket_frt_03_r (PPPPP)"!(
    ComponentPose=@n7
    Alpha=1
    ConstraintWeights_0=0.1633
    ConstraintWeights_1:float=0.3137
    ConstraintWeights_2:float=0.1434
    ConstraintWeights_3:float=0.3059
    ConstraintWeights_4:float=0.0737
   ) -> (Pose) {
    alphaBoolBlend {blendOption=Linear}
    boneToModify {boneName=jacket_frt_03_r}
    constraintSetup [
     {targetBone {boneName=spine_01}}
     {targetBone {boneName=spine_02}}
     {targetBone {boneName=spine_03}}
     {targetBone {boneName=spine_04}}
     {targetBone {boneName=pelvis}}
    ]
    constraintWeights [1,1,1,1,1]
   }
   n9 = "Constraint - jacket_frt_04_r (PP)"!(ComponentPose=@n8, Alpha=1, ConstraintWeights_0=0.2391, ConstraintWeights_1:float=0.7609) -> (Pose) {
    alphaBoolBlend {blendOption=Linear}
    boneToModify {boneName=jacket_frt_04_r}
    constraintSetup [{targetBone {boneName=thigh_r}}, {targetBone {boneName=pelvis}}]
    constraintWeights [1,1]
   }
   n10 = "Constraint - jacket_frt_05_r (PP)"!(ComponentPose=@n9, Alpha=1, ConstraintWeights_0=0.9059, ConstraintWeights_1:float=0.0941) -> (Pose) {
    alphaBoolBlend {blendOption=Linear}
    boneToModify {boneName=jacket_frt_05_r}
    constraintSetup [{targetBone {boneName=thigh_r}}, {targetBone {boneName=pelvis}}]
    constraintWeights [1,1]
   }
   n11 = "Constraint - jacket_frt_06_r (P)"!(ComponentPose=@n10, Alpha=1, ConstraintWeights_0=1, ConstraintWeights_1:float=0) -> (Pose) {
    alphaBoolBlend {blendOption=Linear}
    boneToModify {boneName=jacket_frt_06_r}
    constraintSetup [{targetBone {boneName=thigh_r}}]
    constraintWeights [1,0]
   }
   n12 = "Constraint - jacket_frt_07_r (PP)"!(ComponentPose=@n11, Alpha=1, ConstraintWeights_0=0.9791, ConstraintWeights_1:float=0.0209) -> (Pose) {
    alphaBoolBlend {blendOption=Linear}
    boneToModify {boneName=jacket_frt_07_r}
    constraintSetup [{targetBone {boneName=thigh_r}}, {targetBone {boneName=calf_r}}]
    constraintWeights [1,1]
   }
   n13 = "Constraint - jacket_frt_08_r (PP)"!(ComponentPose=@n12, Alpha=1, ConstraintWeights_0=0.037, ConstraintWeights_1:float=0.963) -> (Pose) {
    alphaBoolBlend {blendOption=Linear}
    boneToModify {boneName=jacket_frt_08_r}
    constraintSetup [{targetBone {boneName=thigh_r}}, {targetBone {boneName=calf_r}}]
    constraintWeights [1,1]
   }
   n14 = "Constraint - jacket_frt_09_r (P)"!(ComponentPose=@n13, Alpha=1, ConstraintWeights_0=1, ConstraintWeights_1:float=0) -> (Pose) {
    alphaBoolBlend {blendOption=Linear}
    boneToModify {boneName=jacket_frt_09_r}
    constraintSetup [{targetBone {boneName=calf_r}}]
    constraintWeights [1,0]
   }
   n15 = "Constraint - jacket_frtSide_02_l (PPPPP)"!(
    ComponentPose=@n14
    Alpha=1
    ConstraintWeights_0=0.2331
    ConstraintWeights_1:float=0.3882
    ConstraintWeights_2:float=0.0505
    ConstraintWeights_3:float=0.1313
    ConstraintWeights_4:float=0.197
   ) -> (Pose) {
    alphaBoolBlend {blendOption=Linear}
    boneToModify {boneName=jacket_frtSide_02_l}
    constraintSetup [
     {targetBone {boneName=spine_01}}
     {targetBone {boneName=spine_02}}
     {targetBone {boneName=spine_03}}
     {targetBone {boneName=spine_04}}
     {targetBone {boneName=pelvis}}
    ]
    constraintWeights [1,1,1,1,1]
   }
   n16 = "Constraint - jacket_frtSide_03_l (PP)"!(ComponentPose=@n15, Alpha=1, ConstraintWeights_0=0.8376, ConstraintWeights_1:float=0.1624) -> (Pose) {
    alphaBoolBlend {blendOption=Linear}
    boneToModify {boneName=jacket_frtSide_03_l}
    constraintSetup [{targetBone {boneName=pelvis}}, {targetBone {boneName=thigh_l}}]
    constraintWeights [1,1]
   }
   n17 = "Constraint - jacket_frtSide_04_l (PP)"!(ComponentPose=@n16, Alpha=1, ConstraintWeights_0=0.1335, ConstraintWeights_1:float=0.8665) -> (Pose) {
    alphaBoolBlend {blendOption=Linear}
    boneToModify {boneName=jacket_frtSide_04_l}
    constraintSetup [{targetBone {boneName=pelvis}}, {targetBone {boneName=thigh_l}}]
    constraintWeights [1,1]
   }
   n18 = "Constraint - jacket_frtSide_05_l (PP)"!(ComponentPose=@n17, Alpha=1, ConstraintWeights_0=0.0134, ConstraintWeights_1:float=0.9866) -> (Pose) {
    alphaBoolBlend {blendOption=Linear}
    boneToModify {boneName=jacket_frtSide_05_l}
    constraintSetup [{targetBone {boneName=pelvis}}, {targetBone {boneName=thigh_l}}]
    constraintWeights [1,1]
   }
   n19 = "Constraint - jacket_frtSide_02_r (PPPPP)"!(
    ComponentPose=@n18
    Alpha=1
    ConstraintWeights_0=0.2289
    ConstraintWeights_1:float=0.3815
    ConstraintWeights_2:float=0.0525
    ConstraintWeights_3:float=0.1321
    ConstraintWeights_4:float=0.2049
   ) -> (Pose) {
    alphaBoolBlend {blendOption=Linear}
    boneToModify {boneName=jacket_frtSide_02_r}
    constraintSetup [
     {targetBone {boneName=spine_01}}
     {targetBone {boneName=spine_02}}
     {targetBone {boneName=spine_03}}
     {targetBone {boneName=spine_04}}
     {targetBone {boneName=pelvis}}
    ]
    constraintWeights [1,1,1,1,1]
   }
   n20 = "Constraint - jacket_frtSide_03_r (PP)"!(ComponentPose=@n19, Alpha=1, ConstraintWeights_0=0.1853, ConstraintWeights_1:float=0.8147) -> (Pose) {
    alphaBoolBlend {blendOption=Linear}
    boneToModify {boneName=jacket_frtSide_03_r}
    constraintSetup [{targetBone {boneName=thigh_r}}, {targetBone {boneName=pelvis}}]
    constraintWeights [1,1]
   }
   n21 = "Constraint - jacket_frtSide_04_r (PP)"!(ComponentPose=@n20, Alpha=1, ConstraintWeights_0=0.8764, ConstraintWeights_1:float=0.1236) -> (Pose) {
    alphaBoolBlend {blendOption=Linear}
    boneToModify {boneName=jacket_frtSide_04_r}
    constraintSetup [{targetBone {boneName=thigh_r}}, {targetBone {boneName=pelvis}}]
    constraintWeights [1,1]
   }
   n22 = "Constraint - jacket_frtSide_05_r (P)"!(ComponentPose=@n21, Alpha=1, ConstraintWeights_0=1, ConstraintWeights_1:float=0) -> (Pose) {
    alphaBoolBlend {blendOption=Linear}
    boneToModify {boneName=jacket_frtSide_05_r}
    constraintSetup [{targetBone {boneName=thigh_r}}]
    constraintWeights [1,0]
   }
   n23 = "Constraint - jacket_side_02_l (PPPPP)"!(
    ComponentPose=@n22
    Alpha=1
    ConstraintWeights_0=0.2426
    ConstraintWeights_1:float=0.334
    ConstraintWeights_2:float=0.0344
    ConstraintWeights_3:float=0.0663
    ConstraintWeights_4:float=0.3227
   ) -> (Pose) {
    alphaBoolBlend {blendOption=Linear}
    boneToModify {boneName=jacket_side_02_l}
    constraintSetup [
     {targetBone {boneName=spine_01}}
     {targetBone {boneName=spine_02}}
     {targetBone {boneName=spine_03}}
     {targetBone {boneName=spine_04}}
     {targetBone {boneName=pelvis}}
    ]
    constraintWeights [1,1,1,1,1]
   }
   n24 = "Constraint - jacket_side_03_l (PP)"!(ComponentPose=@n23, Alpha=1, ConstraintWeights_0=0.8489, ConstraintWeights_1:float=0.1511) -> (Pose) {
    alphaBoolBlend {blendOption=Linear}
    boneToModify {boneName=jacket_side_03_l}
    constraintSetup [{targetBone {boneName=pelvis}}, {targetBone {boneName=thigh_l}}]
    constraintWeights [1,1]
   }
   n25 = "Constraint - jacket_side_04_l (PP)"!(ComponentPose=@n24, Alpha=1, ConstraintWeights_0=0.3356, ConstraintWeights_1:float=0.6644) -> (Pose) {
    alphaBoolBlend {blendOption=Linear}
    boneToModify {boneName=jacket_side_04_l}
    constraintSetup [{targetBone {boneName=pelvis}}, {targetBone {boneName=thigh_l}}]
    constraintWeights [1,1]
   }
   n26 = "Constraint - jacket_side_05_l (PP)"!(ComponentPose=@n25, Alpha=1, ConstraintWeights_0=0.0309, ConstraintWeights_1:float=0.9691) -> (Pose) {
    alphaBoolBlend {blendOption=Linear}
    boneToModify {boneName=jacket_side_05_l}
    constraintSetup [{targetBone {boneName=pelvis}}, {targetBone {boneName=thigh_l}}]
    constraintWeights [1,1]
   }
   n27 = "Constraint - jacket_side_06_l (P)"!(ComponentPose=@n26, Alpha=1, ConstraintWeights_0=1, ConstraintWeights_1:float=0) -> (Pose) {
    alphaBoolBlend {blendOption=Linear}
    boneToModify {boneName=jacket_side_06_l}
    constraintSetup [{targetBone {boneName=thigh_l}}]
    constraintWeights [1,0]
   }
   n28 = "Constraint - jacket_side_07_l (PP)"!(ComponentPose=@n27, Alpha=1, ConstraintWeights_0=0.3489, ConstraintWeights_1:float=0.6511) -> (Pose) {
    alphaBoolBlend {blendOption=Linear}
    boneToModify {boneName=jacket_side_07_l}
    constraintSetup [{targetBone {boneName=calf_l}}, {targetBone {boneName=thigh_l}}]
    constraintWeights [1,1]
   }
   n29 = "Constraint - jacket_side_08_l (PP)"!(ComponentPose=@n28, Alpha=1, ConstraintWeights_0=0.974, ConstraintWeights_1:float=0.026) -> (Pose) {
    alphaBoolBlend {blendOption=Linear}
    boneToModify {boneName=jacket_side_08_l}
    constraintSetup [{targetBone {boneName=calf_l}}, {targetBone {boneName=thigh_l}}]
    constraintWeights [1,1]
   }
   n30 = "Constraint - jacket_side_02_r (PPPPP)"!(
    ComponentPose=@n29
    Alpha=1
    ConstraintWeights_0=0.2507
    ConstraintWeights_1:float=0.3439
    ConstraintWeights_2:float=0.0296
    ConstraintWeights_3:float=0.0635
    ConstraintWeights_4:float=0.3122
   ) -> (Pose) {
    alphaBoolBlend {blendOption=Linear}
    boneToModify {boneName=jacket_side_02_r}
    constraintSetup [
     {targetBone {boneName=spine_01}}
     {targetBone {boneName=spine_02}}
     {targetBone {boneName=spine_03}}
     {targetBone {boneName=spine_04}}
     {targetBone {boneName=pelvis}}
    ]
    constraintWeights [1,1,1,1,1]
   }
   n31 = "Constraint - jacket_side_03_r (PP)"!(ComponentPose=@n30, Alpha=1, ConstraintWeights_0=0.1511, ConstraintWeights_1:float=0.8489) -> (Pose) {
    alphaBoolBlend {blendOption=Linear}
    boneToModify {boneName=jacket_side_03_r}
    constraintSetup [{targetBone {boneName=thigh_r}}, {targetBone {boneName=pelvis}}]
    constraintWeights [1,1]
   }
   n32 = "Constraint - jacket_side_04_r (PP)"!(ComponentPose=@n31, Alpha=1, ConstraintWeights_0=0.6682, ConstraintWeights_1:float=0.3318) -> (Pose) {
    alphaBoolBlend {blendOption=Linear}
    boneToModify {boneName=jacket_side_04_r}
    constraintSetup [{targetBone {boneName=thigh_r}}, {targetBone {boneName=pelvis}}]
    constraintWeights [1,1]
   }
   n33 = "Constraint - jacket_side_05_r (PP)"!(ComponentPose=@n32, Alpha=1, ConstraintWeights_0=0.9865, ConstraintWeights_1:float=0.0135) -> (Pose) {
    alphaBoolBlend {blendOption=Linear}
    boneToModify {boneName=jacket_side_05_r}
    constraintSetup [{targetBone {boneName=thigh_r}}, {targetBone {boneName=pelvis}}]
    constraintWeights [1,1]
   }
   n34 = "Constraint - jacket_side_06_r (P)"!(ComponentPose=@n33, Alpha=1, ConstraintWeights_0=1, ConstraintWeights_1:float=0) -> (Pose) {
    alphaBoolBlend {blendOption=Linear}
    boneToModify {boneName=jacket_side_06_r}
    constraintSetup [{targetBone {boneName=thigh_r}}]
    constraintWeights [1,0]
   }
   n35 = "Constraint - jacket_side_07_r (PP)"!(ComponentPose=@n34, Alpha=1, ConstraintWeights_0=0.6415, ConstraintWeights_1:float=0.3584) -> (Pose) {
    alphaBoolBlend {blendOption=Linear}
    boneToModify {boneName=jacket_side_07_r}
    constraintSetup [{targetBone {boneName=thigh_r}}, {targetBone {boneName=calf_r}}]
    constraintWeights [1,1]
   }
   n36 = "Constraint - jacket_side_08_r (PP)"!(ComponentPose=@n35, Alpha=1, ConstraintWeights_0=0.026, ConstraintWeights_1:float=0.974) -> (Pose) {
    alphaBoolBlend {blendOption=Linear}
    boneToModify {boneName=jacket_side_08_r}
    constraintSetup [{targetBone {boneName=thigh_r}}, {targetBone {boneName=calf_r}}]
    constraintWeights [1,1]
   }
   n37 = "Constraint - jacket_back_02_l (PPPPP)"!(
    ComponentPose=@n36
    Alpha=1
    ConstraintWeights_0=0.2786
    ConstraintWeights_1:float=0.3325
    ConstraintWeights_2:float=0.0691
    ConstraintWeights_3:float=0.026
    ConstraintWeights_4:float=0.2939
   ) -> (Pose) {
    alphaBoolBlend {blendOption=Linear}
    boneToModify {boneName=jacket_back_02_l}
    constraintSetup [
     {targetBone {boneName=spine_01}}
     {targetBone {boneName=spine_02}}
     {targetBone {boneName=spine_03}}
     {targetBone {boneName=spine_04}}
     {targetBone {boneName=pelvis}}
    ]
    constraintWeights [1,1,1,1,1]
   }
   n38 = "Constraint - jacket_back_03_l (PP)"!(ComponentPose=@n37, Alpha=1, ConstraintWeights_0=0.9601, ConstraintWeights_1:float=0.0399) -> (Pose) {
    alphaBoolBlend {blendOption=Linear}
    boneToModify {boneName=jacket_back_03_l}
    constraintSetup [{targetBone {boneName=pelvis}}, {targetBone {boneName=thigh_l}}]
    constraintWeights [1,1]
   }
   n39 = "Constraint - jacket_back_04_l (PP)"!(ComponentPose=@n38, Alpha=1, ConstraintWeights_0=0.5447, ConstraintWeights_1:float=0.4553) -> (Pose) {
    alphaBoolBlend {blendOption=Linear}
    boneToModify {boneName=jacket_back_04_l}
    constraintSetup [{targetBone {boneName=pelvis}}, {targetBone {boneName=thigh_l}}]
    constraintWeights [1,1]
   }
   n40 = "Constraint - jacket_back_05_l (PP)"!(ComponentPose=@n39, Alpha=1, ConstraintWeights_0=0.0682, ConstraintWeights_1:float=0.9318) -> (Pose) {
    alphaBoolBlend {blendOption=Linear}
    boneToModify {boneName=jacket_back_05_l}
    constraintSetup [{targetBone {boneName=pelvis}}, {targetBone {boneName=thigh_l}}]
    constraintWeights [1,1]
   }
   n41 = "Constraint - jacket_back_02_r (PPPPP)"!(
    ComponentPose=@n40
    Alpha=1
    ConstraintWeights_0=0.283
    ConstraintWeights_1:float=0.3464
    ConstraintWeights_2:float=0.072
    ConstraintWeights_3:float=0.0211
    ConstraintWeights_4:float=0.2775
   ) -> (Pose) {
    alphaBoolBlend {blendOption=Linear}
    boneToModify {boneName=jacket_back_02_r}
    constraintSetup [
     {targetBone {boneName=spine_01}}
     {targetBone {boneName=spine_02}}
     {targetBone {boneName=spine_03}}
     {targetBone {boneName=spine_04}}
     {targetBone {boneName=pelvis}}
    ]
    constraintWeights [1,1,1,1,1]
   }
   n42 = "Constraint - jacket_back_03_r (PP)"!(ComponentPose=@n41, Alpha=1, ConstraintWeights_0=0.0399, ConstraintWeights_1:float=0.9601) -> (Pose) {
    alphaBoolBlend {blendOption=Linear}
    boneToModify {boneName=jacket_back_03_r}
    constraintSetup [{targetBone {boneName=thigh_r}}, {targetBone {boneName=pelvis}}]
    constraintWeights [1,1]
   }
   n43 = "Constraint - jacket_back_04_r (PP)"!(ComponentPose=@n42, Alpha=1, ConstraintWeights_0=0.4565, ConstraintWeights_1:float=0.5435) -> (Pose) {
    alphaBoolBlend {blendOption=Linear}
    boneToModify {boneName=jacket_back_04_r}
    constraintSetup [{targetBone {boneName=thigh_r}}, {targetBone {boneName=pelvis}}]
    constraintWeights [1,1]
   }
   n44 = "Constraint - jacket_back_05_r (PP)"!(ComponentPose=@n43, Alpha=1, ConstraintWeights_0=0.9486, ConstraintWeights_1:float=0.0514) -> (Pose) {
    alphaBoolBlend {blendOption=Linear}
    boneToModify {boneName=jacket_back_05_r}
    constraintSetup [{targetBone {boneName=thigh_r}}, {targetBone {boneName=pelvis}}]
    constraintWeights [1,1]
   }
   n45 = "Constraint - jacket_back_02 (PPPPP)"!(
    ComponentPose=@n44
    Alpha=1
    ConstraintWeights_0=0.2641
    ConstraintWeights_1:float=0.3737
    ConstraintWeights_2:float=0.1147
    ConstraintWeights_3:float=0.0276
    ConstraintWeights_4:float=0.2199
   ) -> (Pose) {
    alphaBoolBlend {blendOption=Linear}
    boneToModify {boneName=jacket_back_02}
    constraintSetup [
     {targetBone {boneName=spine_01}}
     {targetBone {boneName=spine_02}}
     {targetBone {boneName=spine_03}}
     {targetBone {boneName=spine_04}}
     {targetBone {boneName=pelvis}}
    ]
    constraintWeights [1,1,1,1,1]
   }
   n46 = "Constraint - jacket_back_03 (P)"!(ComponentPose=@n45, Alpha=1, ConstraintWeights_0=1, ConstraintWeights_1:float=0) -> (Pose) {
    alphaBoolBlend {blendOption=Linear}
    boneToModify {boneName=jacket_back_03}
    constraintSetup [{targetBone {boneName=pelvis}}]
    constraintWeights [1,0]
   }
   n47 = "Constraint - jacket_back_04 (PPP)"!(ComponentPose=@n46, Alpha=1, ConstraintWeights_0=0.1801, ConstraintWeights_1:float=0.7586, ConstraintWeights_2:float=0.0613) -> (Pose) {
    alphaBoolBlend {blendOption=Linear}
    boneToModify {boneName=jacket_back_04}
    constraintSetup [{targetBone {boneName=thigh_r}}, {targetBone {boneName=pelvis}}, {targetBone {boneName=thigh_l}}]
    constraintWeights [1,1,1]
   }
   n48 = "Constraint - jacket_back_05 (PP)"!(ComponentPose=@n47, Alpha=1, ConstraintWeights_0=0.0699, ConstraintWeights_1:float=0.9301) -> (Pose) {
    alphaBoolBlend {blendOption=Linear}
    boneToModify {boneName=jacket_back_05}
    constraintSetup [{targetBone {boneName=pelvis}}, {targetBone {boneName=thigh_l}}]
    constraintWeights [1,1]
   }
   n49 = "Transform (Modify) Bone - Bone: jacket_back_04_l"(ComponentPose=@n48, Alpha=1, Translation="0,0,0", Rotation="0,-20,0", Scale="1,1,1") {
    rotationMode=BMM_Additive
    rotationSpace=BCS_BoneSpace
    alphaBoolBlend {blendOption=Linear}
    boneToModify {boneName=jacket_back_04_l}
    rotation {yaw=-20}
   }
   n50 = "Transform (Modify) Bone - Bone: jacket_back_04_r"(ComponentPose=@n49, Alpha=1, Translation="0,0,0", Rotation="0,-15,0", Scale="1,1,1") {
    rotationMode=BMM_Additive
    rotationSpace=BCS_BoneSpace
    alphaBoolBlend {blendOption=Linear}
    boneToModify {boneName=jacket_back_04_r}
    rotation {yaw=-15}
   }
   n51 = "Transform (Modify) Bone - Bone: jacket_back_06_l"(ComponentPose=@n50, Alpha=Get calf_rot_back_l() -> (calf_rot_back_l:double), Translation="0,0,0", Rotation="0,-35,0", Scale="1,1,1") {
    rotationMode=BMM_Additive
    rotationSpace=BCS_BoneSpace
    alphaBoolBlend {blendOption=Linear}
    boneToModify {boneName=jacket_back_06_l}
    rotation {yaw=-35}
   }
   n52 = "Transform (Modify) Bone - Bone: jacket_back_06_r"(ComponentPose=@n51, Alpha=Get calf_rot_back_r() -> (calf_rot_back_r:double), Translation="0,0,0", Rotation="0,-35,0", Scale="1,1,1") {
    rotationMode=BMM_Additive
    rotationSpace=BCS_BoneSpace
    alphaBoolBlend {blendOption=Linear}
    boneToModify {boneName=jacket_back_06_r}
    rotation {yaw=-35}
   }
   n53 = "Sequence Player (additive) 'M_MED_Star_Walker_Reset_Pose'"() {sequence=&/M_MED_Star_Walker_Reset_Pose}
   n54 = Use cached pose 'AnimCache'()
   n55 = RigidBody(
    ComponentPose=Local To Component(LocalPose=Apply Additive(Base=@n54, Additive=@n53, Alpha=0.8) {alphaBoolBlend {blendOption=Linear}})
    Alpha=Get Is Not DBNOCarried() -> (Is Not DBNOCarried:double)
   ) {
    bClampLinearTranslationLimitToRefPose=true
    bOverrideWorldGravity=true
    overridePhysicsAsset=&/M_MED_Star_Walker_Coat_Physics
    simulationSpace=BaseBoneSpace
    alphaBoolBlend {blendOption=Linear}
    baseBoneRef {boneName=spine_04}
    componentAppliedLinearAccClamp=<1000 1000 1000>
    overrideWorldGravity {z=-980}
    simSpaceSettings {
     dampingAlpha=0.25
     maxAngularAcceleration=40
     maxAngularVelocity=8
     maxLinearAcceleration=3500
     maxLinearVelocity=700
     velocityScaleZ=0.25
     worldAlpha=1
    }
   }
   n56 = RigidBody(ComponentPose=@n55) {
    bClampLinearTranslationLimitToRefPose=true
    bOverrideWorldGravity=true
    overridePhysicsAsset=&/M_MED_Star_Walker_Collar_Physics
    simulationSpace=BaseBoneSpace
    alphaBoolBlend {blendOption=Linear}
    baseBoneRef {boneName=root}
    componentAppliedLinearAccClamp=<1000 1000 1300>
    componentLinearAccScale=<0.8 0.8 0.6>
    componentLinearVelScale=<0.8 0.8 0.6>
    overrideWorldGravity {z=-950}
   }
   n57 = Component To Local(ComponentPose=@n56)
   n58 = RigidBody(ComponentPose=Local To Component(LocalPose=Use cached pose 'AnimCache'()), ExternalForce=Get skydiving_force() -> (skydiving_force:Vector)) {
    bClampLinearTranslationLimitToRefPose=true
    bOverrideWorldGravity=true
    overridePhysicsAsset=&M_MED_Star_Walker_Coat_Physics
    simulationSpace=BaseBoneSpace
    alphaBoolBlend {blendOption=Linear}
    baseBoneRef {boneName=pelvis}
    componentAppliedLinearAccClamp=<900 600 600>
    componentLinearAccScale=<0.1 0.1 0.1>
    componentLinearVelScale=<0.1 0.1 0.1>
    overrideWorldGravity {z=-150}
   }
   n59 = Component To Local(ComponentPose=@n58)
   n60 = Get bIsSkyDiving()
   n61 = RigidBody(
    ComponentPose=Local To Component(LocalPose=Blend Poses by bool(BlendPose_0=@n59, BlendPose_1=@n57, BlendTime_0=0.3, BlendTime_1=0.3, bActiveValue=@n60) {blendType=Linear})
    Alpha=Get Is DBNOCarried() -> (Is DBNOCarried:double)
   ) {
    bClampLinearTranslationLimitToRefPose=true
    bOverrideWorldGravity=true
    overridePhysicsAsset=&/M_MED_Star_Walker_Coat_Physics_DBNOcarried
    simulationSpace=BaseBoneSpace
    alphaBoolBlend {blendOption=Linear}
    baseBoneRef {boneName=root}
    componentAppliedLinearAccClamp=<1000 1000 1000>
    componentLinearAccScale=<0.6 0.6 0.1>
    componentLinearVelScale=<0.6 0.6 0.1>
    overrideWorldGravity {z=-725}
   }
   Output Pose(Result=Blend Poses by bool(BlendPose_0=Component To Local(ComponentPose=@n61), BlendPose_1=Component To Local(ComponentPose=@n52), bActiveValue=Get bRigidBodyEnabled()) {blendType=Linear}) {layerGroup=None}
   n62 = "Transform (Modify) Bone - Bone: jacket_frt_04_l"(
    ComponentPose=Local To Component(LocalPose=Copy Pose From Mesh() {bUseAttachedParent=true})
    Alpha=1
    Translation="-0.067,2.647,4.771"
    Rotation="0,0,0"
    Scale="1,1,1"
   ) {
    translationMode=BMM_Additive
    translationSpace=BCS_BoneSpace
    alphaBoolBlend {blendOption=Linear}
    boneToModify {boneName=jacket_frt_04_l}
    translation=<-0.067 2.647 4.771>
   }
   n63 = "Transform (Modify) Bone - Bone: jacket_frtSide_04_l"(ComponentPose=@n62, Alpha=1, Translation="0.052,2.885,2.091", Rotation="0,0,0", Scale="1,1,1") {
    translationMode=BMM_Additive
    translationSpace=BCS_BoneSpace
    alphaBoolBlend {blendOption=Linear}
    boneToModify {boneName=jacket_frtSide_04_l}
    translation=<0.052 2.885 2.091>
   }
   n64 = "Constraint - dyn_parent_shldr_tipA_l (PPPPPPPP)"!(
    ComponentPose=@n63
    Alpha=1
    ConstraintWeights_0=0.1879
    ConstraintWeights_1:float=0.1743
    ConstraintWeights_2:float=0.0866
    ConstraintWeights_3:float=0.0578
    ConstraintWeights_4:float=0.0894
    ConstraintWeights_5:float=0.2478
    ConstraintWeights_6:float=0.1352
    ConstraintWeights_7:float=0.0211
   ) -> (Pose) {
    alphaBoolBlend {blendOption=Linear}
    boneToModify {boneName=dyn_parent_shldr_tipA_l}
    constraintSetup [
     {targetBone {boneName=head}}
     {targetBone {boneName=clavicle_l}}
     {targetBone {boneName=upperarm_twist_02_l}}
     {targetBone {boneName=spine_03}}
     {targetBone {boneName=spine_04}}
     {targetBone {boneName=spine_05}}
     {targetBone {boneName=upperarm_twist_01_l}}
     {targetBone {boneName=neck_02}}
    ]
    constraintWeights [1,1,1,1,1,1,1,1]
   }
   n65 = "Constraint - dyn_parent_shldr_tipB_l (PPPPPPPP)"!(
    ComponentPose=@n64
    Alpha=1
    ConstraintWeights_0=0.1084
    ConstraintWeights_1:float=0.2337
    ConstraintWeights_2:float=0.1404
    ConstraintWeights_3:float=0.0548
    ConstraintWeights_4:float=0.0716
    ConstraintWeights_5:float=0.1791
    ConstraintWeights_6:float=0.1949
    ConstraintWeights_7:float=0.0173
   ) -> (Pose) {
    alphaBoolBlend {blendOption=Linear}
    boneToModify {boneName=dyn_parent_shldr_tipB_l}
    constraintSetup [
     {targetBone {boneName=head}}
     {targetBone {boneName=clavicle_l}}
     {targetBone {boneName=upperarm_twist_02_l}}
     {targetBone {boneName=spine_03}}
     {targetBone {boneName=spine_04}}
     {targetBone {boneName=spine_05}}
     {targetBone {boneName=upperarm_twist_01_l}}
     {targetBone {boneName=upperarm_l}}
    ]
    constraintWeights [1,1,1,1,1,1,1,1]
   }
   n66 = "Constraint - dyn_parent_cowl_l (PPPPPPPP)"!(
    ComponentPose=@n65
    Alpha=1
    ConstraintWeights_0=0.0439
    ConstraintWeights_1:float=0.2485
    ConstraintWeights_2:float=0.0943
    ConstraintWeights_3:float=0.0841
    ConstraintWeights_4:float=0.1478
    ConstraintWeights_5:float=0.1871
    ConstraintWeights_6:float=0.1077
    ConstraintWeights_7:float=0.0867
   ) -> (Pose) {
    alphaBoolBlend {blendOption=Linear}
    boneToModify {boneName=dyn_parent_cowl_l}
    constraintSetup [
     {targetBone {boneName=head}}
     {targetBone {boneName=clavicle_l}}
     {targetBone {boneName=upperarm_twist_02_l}}
     {targetBone {boneName=spine_02}}
     {targetBone {boneName=spine_03}}
     {targetBone {boneName=spine_04}}
     {targetBone {boneName=spine_05}}
     {targetBone {boneName=upperarm_twist_01_l}}
    ]
    constraintWeights [1,1,1,1,1,1,1,1]
   }
   n67 = "Constraint - dyn_parent_cowl_r (PPPPPPPP)"!(
    ComponentPose=@n66
    Alpha=1
    ConstraintWeights_0=0.0369
    ConstraintWeights_1:float=0.0852
    ConstraintWeights_2:float=0.0983
    ConstraintWeights_3:float=0.0863
    ConstraintWeights_4:float=0.1512
    ConstraintWeights_5:float=0.1898
    ConstraintWeights_6:float=0.1041
    ConstraintWeights_7:float=0.2482
   ) -> (Pose) {
    alphaBoolBlend {blendOption=Linear}
    boneToModify {boneName=dyn_parent_cowl_r}
    constraintSetup [
     {targetBone {boneName=head}}
     {targetBone {boneName=upperarm_twist_01_r}}
     {targetBone {boneName=upperarm_twist_02_r}}
     {targetBone {boneName=spine_02}}
     {targetBone {boneName=spine_03}}
     {targetBone {boneName=spine_04}}
     {targetBone {boneName=spine_05}}
     {targetBone {boneName=clavicle_r}}
    ]
    constraintWeights [1,1,1,1,1,1,1,1]
   }
   Save cached pose 'AnimCache'(Pose=Component To Local(ComponentPose=@n67)) {cacheName=AnimCache}
  ]}
 EventGraph {nodes [
   Event BlueprintUpdateAnimation()
   > Get!(execute:Exec) -> (then:Exec, else:Exec, Skeletal_Mesh:Skeletal Mesh Component Object Reference)
   > n1 = Sequence!(execute) -> (then_0, then_1, then_2:Exec, then_3:Exec, then_4:Exec, then_5:Exec, then_6:Exec)
   n2 = Get bIsDBNOCarried()
   Set Is DBNOCarried(execute=@n1.then_4, Is DBNOCarried:double="To Float (Boolean)"(InBool=@n2)) -> (Output_Get:double)
   > Set Is Not DBNOCarried(Is Not DBNOCarried:double="To Float (Boolean)"(InBool=NOT Boolean(A=@n2))) -> (Output_Get:double)
   n3 = Branch(execute=@n1.then_0, Condition=Get bIsSkyDiving())
   n4 = Break Fort Anim Input Skydiving External Force(FortAnimInput_SkydivingExternalForce=Get SkydivingExternalForceInput()) -> (
    bUseSkydivingVectorForce=~
    SkydivingVectorForce=~
    HeadToPelvisDirection=~
    FloatingMultiplier=~
    FloatingAdditive=~
    DivingMultiplier=~
    DivingAdditive=~
    ParachutingMultiplier=~
    ParachutingAdditive=~
   )
   > Set skydiving_force(skydiving_force:Vector=vector * float(A=@n4.SkydivingVectorForce, B=0.65)) -> (Output_Get:Vector)
   > Branch(Condition=Get bIsParachuteOpened())
   > Set skydiving_force(skydiving_force=vector * float(A=@n4.SkydivingVectorForce, B=0.1))
   Set skydiving_force(execute=@n3.else, skydiving_force="0, 0, 0")
   n5 = BreakTransform(InTransform=GetDeltaTransformFromRefPose(self=Get Skeletal_Mesh() -> (Skeletal_Mesh:Skeletal Mesh Component Object Reference), BoneName=calf_l, BaseName=thigh_l))
   Set calf_rot_back_l(execute=@n1.then_3, calf_rot_back_l:double=MapRangeClamped(Value=BreakRotator(InRot=@n5.Rotation).Yaw, InRangeA=15, InRangeB=115)) -> (Output_Get:double)
   n6 = Get CurrentLOD()
   n7 = Get bRigidBodyEnabled()
   Set rigids_off(execute=@n1.then_1, rigids_off:bool=OR Boolean(A=NOT Boolean(A=@n7), B="integer > integer"(A=@n6, B=1) {is=title K2Node_CallFunction})) -> (Output_Get:bool)
   Branch(execute=@n1.then_6, Condition=Get fromDBNO() -> (fromDBNO:bool))
   > Branch(Condition=NOT Boolean(A=Get bIsDBNO()))
   > Delay(Duration=0.15)
   > ResetDynamics(InTeleportType=TeleportPhysics)
   > ResetDynamics(InTeleportType=ResetPhysics)
   > n8 = Set fromDBNO(fromDBNO:bool=false) -> (Output_Get:bool)
   Do Once(execute=@n10.then, Reset=@n8.then)
   > Set fromDBNO(fromDBNO=true)
   n9 = BreakTransform(InTransform=GetDeltaTransformFromRefPose(self=Get Skeletal_Mesh(), BoneName=calf_r, BaseName=thigh_r))
   Set calf_rot_back_r(execute=@n1.then_2, calf_rot_back_r:double=MapRangeClamped(Value=BreakRotator(InRot=@n9.Rotation).Yaw, InRangeA=15, InRangeB=115)) -> (Output_Get:double)
   n10 = Branch(execute=@n1.then_5, Condition=Get bIsDBNO())
   Event OnParachuteOpened()
   > n11 = Delay(Duration=1.5)
   ResetDynamics(execute=[@n11.then, @n12.then, @n13.then], InTeleportType=TeleportPhysics)
   > ResetDynamics(InTeleportType=ResetPhysics)
   n12 = Event OnEndSkydiving()
   Event BlueprintInitializeAnimation()
   > ResetDynamics(InTeleportType=TeleportPhysics)
   > Set Skeletal_Mesh(Skeletal_Mesh:Skeletal Mesh Component Object Reference=GetOwnerSkeletalMeshComponent()) -> (Output_Get:Skeletal Mesh Component Object Reference)
   n13 = Event OnBeginSkydiving()
  ]}
]
variables [
 calf_rot_back_l {type=number}
 calf_rot_back_r {type=number}
 fromDBNO {type=boolean}
 Is DBNOCarried {type=number}
 Is Not DBNOCarried {type=number}
 rigids_off {type=boolean}
 Skeletal_Mesh {type=SkeletalMeshComponent}
 skydiving_force {type=Vector}
]
