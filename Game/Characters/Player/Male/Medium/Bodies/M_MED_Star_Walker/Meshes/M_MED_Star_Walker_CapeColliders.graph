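# Physics asset holding the cape collision volumes for the medium male Star Walker player body, as its CapeColliders name indicates.
# Capsules cover the pelvis, spine, neck, clavicles and four jacket back and side bones, with a sphere on the head.
# Arm, leg and foot bodies carry no shapes.
# It defines no constraints, so the bodies form an unjointed collider set rather than a ragdoll.
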PhysicsAsset /Game/Characters/Player/Male/Medium/Bodies/M_MED_Star_Walker/Meshes/M_MED_Star_Walker_CapeColliders
bodies {
 calf_l []
 calf_r []
 clavicle_l [{shapeType=Capsule, length=7.7636, radius=11.1427, center=<9.9642 -0.7741 -1.5923>, rotation=<rot 75.9352 -0.0001 175.5622>}]
 clavicle_r [{shapeType=Capsule, length=7.7636, radius=11.1427, center=<-9.9642 0.7741 1.5923>, rotation=<rot 75.9351 -0.0001 -4.4378>}]
 foot_l []
 foot_r []
 head [{shapeType=Sphere, radius=17.3125, center=<11.028 -0.174 0>}]
 jacket_back_08_l [{shapeType=Capsule, length=25, radius=9.5, shapeName=jacket_back_08_l_capsule, center=<17.4328 -3.9219 0>, rotation=<rot 90 -90 -90.0001>}]
 jacket_back_08_r [{shapeType=Capsule, length=25.5, radius=7.75, shapeName=jacket_back_08_r_capsule, center=<17.1617 -1.7028 0>, rotation=<rot 90 180 179.9997>}]
 jacket_side_08_l [{
   shapeType=Capsule
   length=22.25
   radius=7.75
   shapeName=jacket_side_08_l_capsule
   center=<9.6734 -1.6631 0.3639>
   rotation=<rot -88.2657 85.0841 -94.9325>
  }]
 jacket_side_08_r [{shapeType=Capsule, length=24, radius=8.25, shapeName=jacket_side_08_r_capsule, center=<9.9677 -3.7357 0>, rotation=<rot -90 90 -89.9999>}]
 lowerarm_l []
 lowerarm_r []
 neck_02 [{shapeType=Capsule, length=13.4375, radius=7.3175, center=<-0.2642 -0.3408 0>, rotation=<rot 90 -171.1193 -175.3978>}]
 pelvis [{shapeType=Capsule, length=7.4062, radius=15.2888, center=<-4.6483 -2.6271 0>}]
 spine_01 [{shapeType=Capsule, length=10.7187, radius=12.4138, center=<-2.6215 -3.3065 0>}]
 spine_02 [
  {shapeType=Capsule, length=8.5937, radius=12.945, center=<-2.6215 -0.7823 0.0001>}
  {shapeType=Capsule, length=29.875, radius=11.1013, center=<2.5024 0.4992 3.9825>, rotation=<rot 90 0.0049 0.0049>}
  {shapeType=Capsule, length=29.875, radius=11.1013, center=<2.4704 0.4991 -3.9825>, rotation=<rot 90 -0.0396 -0.0396>}
 ]
 spine_03 [{shapeType=Capsule, length=7.4062, radius=12.5075, center=<-1.953 0.1364 0.0001>}]
 spine_04 [{shapeType=Capsule, length=14.6562, radius=13.6325, center=<2.1467 -1.2748 0.0002>}]
 spine_05 [
  {shapeType=Capsule, length=17.8437, radius=17.195, center=<-1.7514 2.4879 0>}
  {shapeType=Capsule, length=12.7812, radius=17.0625, center=<0.7875 3.1143 0>}
 ]
 thigh_l []
 thigh_r []
 upperarm_l []
 upperarm_r []
}
constraints []
